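# Physics asset for the Ruckus two-handed pickaxe, defining the bodies and joints for its two dynamic "shoulder" bones.
# A long capsule covers the handle bone, and a thin tilted capsule covers each of the left and right dyn shoulder bones.
# Each shoulder is constrained directly to the handle with limited swing and twist.
# The swing limits are asymmetric, differing between the left and right sides.
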PhysicsAsset /BRCosmetics/Weapons/FORT_Melee/Pickaxe_Ruckus/Meshes/Ruckus_Axe_Physics
bodies {
 Melee_TwoHanded_Handle [{shapeType=Capsule, length=100, radius=5, center=<0 0 22.5703>}]
 dyn_L_shoulder [{shapeType=Capsule, length=15, radius=1, center=<-7.6677 1.8114 3.1188>, rotation=<rot 50 0 0>}]
 dyn_R_shoulder [{shapeType=Capsule, length=15, radius=1, center=<5.5151 0.0561 -4.5527>, rotation=<rot 50 0 0>}]
}
constraints [
 {bone1Name=dyn_R_shoulder, bone2Name=Melee_TwoHanded_Handle, swing1LimitDegrees=40, swing2LimitDegrees=15, twistLimitDegrees=15}
 {bone1Name=dyn_L_shoulder, bone2Name=Melee_TwoHanded_Handle, swing1LimitDegrees=30, swing2LimitDegrees=20, twistLimitDegrees=15}
]
