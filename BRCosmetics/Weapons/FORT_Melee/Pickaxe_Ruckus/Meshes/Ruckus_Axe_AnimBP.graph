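# Animation Blueprint for the Ruckus pickaxe mesh that adds physics-driven secondary motion to its skeleton.
# The input pose runs through a RigidBody node using the Ruckus_Axe_Physics override asset, with damped-down component acceleration and velocity, then goes to the output.
# A chained pair of Anim Dynamics nodes for the left and right shoulder bones sits disconnected from the output.
# The EventGraph holds only unwired stub nodes, with no variables.
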AnimBlueprint /BRCosmetics/Weapons/FORT_Melee/Pickaxe_Ruckus/Meshes/Ruckus_Axe_AnimBP
target_skeleton=&/Ruckus_Axe_Skeleton
graphs [
 AnimGraph {nodes [
   n1 = Anim Dynamics - Bone: dyn_L_shoulder(Alpha=1, ExternalForce="0,0,0") {
    angularDampingOverride=1
    angularSpringConstant=100
    bGravityOverrideInSimSpace=true
    bOverrideAngularDamping=true
    gravityScale=0.55
    alphaBoolBlend {blendOption=Linear}
    boundBone {boneName=dyn_L_shoulder}
    componentAppliedLinearAccClamp=<4000 4000 4000>
    componentLinearAccScale=<0.55 0.55 0.55>
    physicsBodyDefinitions [{
      sphereCollisionRadius=0
      boundBone {boneName=dyn_L_shoulder}
      boxExtents=<1 1 5>
      constraintSetup {angularLimitsMax=<15 0 10>, angularLimitsMin=<-15 0 -10>, angularTarget=<0 0 0>}
      localJointOffset=<-0.1 -0.1 0>
     }]
   }
   Anim Dynamics - Bone: dyn_R_shoulder(ComponentPose=@n1, Alpha=1, ExternalForce="0,0,0") {
    angularDampingOverride=1
    angularSpringConstant=100
    bGravityOverrideInSimSpace=true
    bOverrideAngularDamping=true
    gravityScale=0.4
    alphaBoolBlend {blendOption=Linear}
    boundBone {boneName=dyn_R_shoulder}
    componentAppliedLinearAccClamp=<4000 4000 4000>
    componentLinearAccScale=<0.55 0.55 0.4>
    physicsBodyDefinitions [{
      sphereCollisionRadius=0
      boundBone {boneName=dyn_R_shoulder}
      boxExtents=<1 1 5>
      constraintSetup {angularLimitsMax=<15 0 10>, angularLimitsMin=<-15 0 -10>, angularTarget=<0 0 0>}
      localJointOffset=<-0.1 -0.1 0>
     }]
   }
   AnimGraphNode_SubInput_0 = Input Pose()
   n2 = RigidBody(ComponentPose=Local To Component(LocalPose=@AnimGraphNode_SubInput_0)) {
    bClampLinearTranslationLimitToRefPose=true
    overridePhysicsAsset=&/Ruckus_Axe_Physics
    alphaBoolBlend {blendOption=Linear}
    componentAppliedLinearAccClamp=<4000 4000 4000>
    componentLinearAccScale=<0.3 0.3 0.3>
    componentLinearVelScale=<0.2 0.2 0.2>
   }
   Output Pose(Result=Component To Local(ComponentPose=@n2)) {layerGroup=None}
  ]}
 EventGraph {nodes [
   TryGetPawnOwner()
   Event BlueprintUpdateAnimation()
  ]}
]
variables []
